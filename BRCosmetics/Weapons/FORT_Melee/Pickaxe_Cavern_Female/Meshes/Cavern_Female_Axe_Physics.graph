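# Physics asset for the Cavern Female pickaxe, giving its dangling base whip chain simulated motion on the two-handed weapon.
# A capsule on the handle bone anchors a small master body through a joint with free swing and twist.
# Five small capsule segments chain from that master body, each constrained to the previous one.
# The first link allows wide swing with almost no twist, while later links share tighter swing and moderate twist limits.
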PhysicsAsset /BRCosmetics/Weapons/FORT_Melee/Pickaxe_Cavern_Female/Meshes/Cavern_Female_Axe_Physics
bodies {
 Melee_TwoHanded_Handle [{shapeType=Capsule, length=87.0188, radius=4.1442, center=<0 0 20.7056>}]
 dyn_baseWhip_01 [{shapeType=Capsule, length=3, radius=3, center=<4.1925 0 -0.0002>, rotation=<rot 90 0 0>}]
 dyn_baseWhip_02 [{shapeType=Capsule, length=3, radius=3, center=<4.1925 0 -0.0002>, rotation=<rot 90 0 0>}]
 dyn_baseWhip_03 [{shapeType=Capsule, length=3, radius=3, center=<4.1925 0 -0.0002>, rotation=<rot 90 0 0>}]
 dyn_baseWhip_04 [{shapeType=Capsule, length=3, radius=3, center=<4.1925 0 -0.0002>, rotation=<rot 90 0 0>}]
 dyn_baseWhip_05 [{shapeType=Capsule, length=3, radius=3, center=<4.1925 0 -0.0002>, rotation=<rot 90 0 0>}]
 dyn_baseWhip_MSTR [{shapeType=Capsule, length=0.505, radius=0.505}]
}
constraints [
 {
  bone1Name=dyn_baseWhip_MSTR
  bone2Name=Melee_TwoHanded_Handle
  swing1LimitDegrees=45
  swing1Motion=Free
  swing2LimitDegrees=45
  swing2Motion=Free
  twistLimitDegrees=45
  twistMotion=Free
 }
 {bone1Name=dyn_baseWhip_01, bone2Name=dyn_baseWhip_MSTR, swing1LimitDegrees=80, swing2LimitDegrees=80, twistLimitDegrees=1}
 {bone1Name=dyn_baseWhip_02, bone2Name=dyn_baseWhip_01, swing1LimitDegrees=60, swing2LimitDegrees=60, twistLimitDegrees=20}
 {bone1Name=dyn_baseWhip_03, bone2Name=dyn_baseWhip_02, swing1LimitDegrees=60, swing2LimitDegrees=60, twistLimitDegrees=20}
 {bone1Name=dyn_baseWhip_04, bone2Name=dyn_baseWhip_03, swing1LimitDegrees=60, swing2LimitDegrees=60, twistLimitDegrees=20}
 {bone1Name=dyn_baseWhip_05, bone2Name=dyn_baseWhip_04, swing1LimitDegrees=60, swing2LimitDegrees=60, twistLimitDegrees=20}
]
