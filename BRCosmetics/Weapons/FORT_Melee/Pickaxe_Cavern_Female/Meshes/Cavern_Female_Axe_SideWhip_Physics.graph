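# Physics asset for the Cavern Female two-handed pickaxe, simulating the dangling side whip on the axe.
# A capsule covers the handle, and three thin box bodies form a chain from it, each joint carrying the same narrow swing and twist limits.
# Small top and bottom capsules attach directly to the handle with free swing and twist, so their stored angle limits do not restrict them.
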PhysicsAsset /BRCosmetics/Weapons/FORT_Melee/Pickaxe_Cavern_Female/Meshes/Cavern_Female_Axe_SideWhip_Physics
bodies {
 Melee_TwoHanded_Handle [{shapeType=Capsule, length=87.0188, radius=4.1442, center=<0 0 20.7056>}]
 dyn_sideWhip_01 [{shapeType=Box, extentX=11.8251, extentY=22.311, extentZ=1.01, center=<6.3988 0 0.0001>}]
 dyn_sideWhip_02 [{shapeType=Box, extentX=10.3419, extentY=22.311, extentZ=1.01, center=<4.841 0 0>}]
 dyn_sideWhip_03 [{shapeType=Box, extentX=10.4279, extentY=20.0719, extentZ=1.01, center=<4.8476 0 0.0001>}]
 dyn_sideWhip_btm [{shapeType=Capsule, length=2, radius=2}]
 dyn_sideWhip_top [{shapeType=Capsule, length=2, radius=2}]
}
constraints [
 {bone1Name=dyn_sideWhip_01, bone2Name=Melee_TwoHanded_Handle, swing1LimitDegrees=18, swing2LimitDegrees=18, twistLimitDegrees=18}
 {bone1Name=dyn_sideWhip_02, bone2Name=dyn_sideWhip_01, swing1LimitDegrees=18, swing2LimitDegrees=18, twistLimitDegrees=18}
 {bone1Name=dyn_sideWhip_03, bone2Name=dyn_sideWhip_02, swing1LimitDegrees=18, swing2LimitDegrees=18, twistLimitDegrees=18}
 {
  bone1Name=dyn_sideWhip_top
  bone2Name=Melee_TwoHanded_Handle
  swing1LimitDegrees=65
  swing1Motion=Free
  swing2LimitDegrees=37
  swing2Motion=Free
  twistLimitDegrees=10
  twistMotion=Free
 }
 {
  bone1Name=dyn_sideWhip_btm
  bone2Name=Melee_TwoHanded_Handle
  swing1LimitDegrees=65
  swing1Motion=Free
  swing2LimitDegrees=37
  swing2Motion=Free
  twistLimitDegrees=10
  twistMotion=Free
 }
]
