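# Animation Blueprint for the Cavern Female pickaxe, adding secondary physics motion to its whip bones.
# The pose copied from a mesh gets an additive dyn_baseWhip_MSTR offset, then passes through two RigidBody nodes with separate physics assets, simulated relative to b_AR_Root with custom gravity.
# Partial-alpha LookAts then aim the side-whip top and bottom bones at dyn_sideWhip_03.
# A side-whip constraint chain is disconnected from the output, and the variables are unused.
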AnimBlueprint /BRCosmetics/Weapons/FORT_Melee/Pickaxe_Cavern_Female/Meshes/Cavern_Female_Axe_AnimBP
target_skeleton=&/Cavern_Female_Axe_Skeleton
graphs [
 AnimGraph {nodes [
   n1 = "Constraint - dyn_sideWhip_01 (P)"(Alpha=0.3, ConstraintWeights_0=1) {
    alphaBoolBlend {blendOption=Linear}
    boneToModify {boneName=dyn_sideWhip_01}
    constraintSetup [{targetBone {boneName=Melee_TwoHanded_Handle}}]
    constraintWeights [1]
   }
   n2 = "Constraint - dyn_sideWhip_02 (P)"(ComponentPose=@n1, Alpha=0.4, ConstraintWeights_0=1) {
    alphaBoolBlend {blendOption=Linear}
    boneToModify {boneName=dyn_sideWhip_02}
    constraintSetup [{targetBone {boneName=Melee_TwoHanded_Handle}}]
    constraintWeights [1]
   }
   n3 = "Constraint - dyn_sideWhip_03 (P)"(ComponentPose=@n2, Alpha=0.5, ConstraintWeights_0=1) {
    alphaBoolBlend {blendOption=Linear}
    boneToModify {boneName=dyn_sideWhip_03}
    constraintSetup [{targetBone {boneName=Melee_TwoHanded_Handle}}]
    constraintWeights [1]
   }
   n4 = "Constraint - dyn_sideWhip_04 (P)"(ComponentPose=@n3, Alpha=0.6, ConstraintWeights_0=1) {
    alphaBoolBlend {blendOption=Linear}
    boneToModify {boneName=dyn_sideWhip_04}
    constraintSetup [{targetBone {boneName=Melee_TwoHanded_Handle}}]
    constraintWeights [1]
   }
   "Constraint - dyn_sideWhip_05 (P)"(ComponentPose=@n4, Alpha=1, ConstraintWeights_0=1) {
    alphaBoolBlend {blendOption=Linear}
    boneToModify {boneName=dyn_sideWhip_05}
    constraintSetup [{targetBone {boneName=Melee_TwoHanded_Handle}}]
    constraintWeights [1]
   }
   n5 = "Transform (Modify) Bone - Bone: dyn_baseWhip_MSTR"(
    ComponentPose=Local To Component(LocalPose=Copy Pose From Mesh())
    Alpha=1
    Translation="1.007,-0.130,-0.254"
    Rotation="0.000,40.086,0.000"
    Scale="1,1,1"
   ) {
    rotationMode=BMM_Additive
    scaleMode=BMM_Additive
    translationMode=BMM_Additive
    alphaBoolBlend {blendOption=Linear}
    boneToModify {boneName=dyn_baseWhip_MSTR}
    rotation {yaw=40.086}
    translation=<1.007 -0.13 -0.254>
   }
   n6 = RigidBody(ComponentPose=@n5) {
    bClampLinearTranslationLimitToRefPose=true
    bOverrideWorldGravity=true
    overridePhysicsAsset=&/Cavern_Female_Axe_Physics
    simulationSpace=BaseBoneSpace
    alphaBoolBlend {blendOption=Linear}
    baseBoneRef {boneName=b_AR_Root}
    componentLinearAccScale=<0.4 0.4 0.42>
    overrideWorldGravity {z=-1050}
   }
   n7 = RigidBody(ComponentPose=@n6) {
    bClampLinearTranslationLimitToRefPose=true
    bOverrideWorldGravity=true
    overridePhysicsAsset=&/Cavern_Female_Axe_SideWhip_Physics
    simulationSpace=BaseBoneSpace
    alphaBoolBlend {blendOption=Linear}
    baseBoneRef {boneName=b_AR_Root}
    componentAppliedLinearAccClamp=<2000 2000 2000>
    componentLinearAccScale=<0.6 2 0.4>
    overrideWorldGravity {z=-1050}
   }
   n8 = Look At - Bone: dyn_sideWhip_top(ComponentPose=@n7, Alpha=0.5) {
    alphaBoolBlend {blendOption=Linear}
    boneToModify {boneName=dyn_sideWhip_top}
    lookAtLocation {x=0}
    lookAtTarget {boneReference {boneName=dyn_sideWhip_03}}
    lookAt_Axis {axis=<1 0>}
    lookUp_Axis {axis=<0 1>}
   }
   n9 = Look At - Bone: dyn_sideWhip_btm(ComponentPose=@n8, Alpha=0.3) {
    alphaBoolBlend {blendOption=Linear}
    boneToModify {boneName=dyn_sideWhip_btm}
    lookAtLocation {x=0}
    lookAtTarget {boneReference {boneName=dyn_sideWhip_03}}
    lookAt_Axis {axis=<1 0>}
    lookUp_Axis {axis=<0 1>}
   }
   Output Pose(Result=Component To Local(ComponentPose=@n9)) {layerGroup=None}
  ]}
 EventGraph {nodes [
   TryGetPawnOwner()
   Event BlueprintUpdateAnimation()
  ]}
]
variables [
 BodyRootLocation {type=Vector}
 BodyRootRotation {type=Rotator}
 CalfLeftLocation {type=Vector}
 CalfLeftRotation {type=Rotator}
 CalfRightLocation {type=Vector}
 CalfRightRotation {type=Rotator}
 ClavicleLeftLocation {type=Vector}
 ClavicleLeftRotation {type=Rotator}
 ClavicleRightLocation {type=Vector}
 ClavicleRightRotation {type=Rotator}
 isEmotingAlpha {type=number}
 LowerarmLeftLocation {type=Vector}
 LowerarmLeftRotation {type=Rotator}
 LowerarmRightLocation {type=Vector}
 LowerarmRightRotation {type=Rotator}
 PelvisLocation {type=Vector}
 PelvisRotation {type=Rotator}
 Spine01Location {type=Vector}
 Spine01Rotation {type=Rotator}
 Spine02Location {type=Vector}
 Spine02Rotation {type=Rotator}
 Spine03Location {type=Vector}
 Spine03Rotation {type=Rotator}
 Spine04Location {type=Vector}
 Spine04Rotation {type=Rotator}
 Spine05Location {type=Vector}
 Spine05Rotation {type=Rotator}
 ThighLeftLocation {type=Vector}
 ThighLeftRotation {type=Rotator}
 ThighRightLocation {type=Vector}
 ThighRightRotation {type=Rotator}
 UpperarmLeftLocation {type=Vector}
 UpperarmLeftRotation {type=Rotator}
 UpperarmRightLocation {type=Vector}
 UpperarmRightRotation {type=Rotator}
]
